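# Unlit emissive material for the RadarTowerV2 build base effect, giving a flickering, tinted glow with drifting cloud detail.
# Multiplied sine and cosine waves of short, differently looping Time values blend EmissiveColor between a dimmed and full version.
# Two cloud texture samples, panned in opposite directions at different tilings, are multiplied together and added to that color.
# The EmissiveScalar parameter multiplies the sum to control overall intensity.
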Material /Game/Blueprints/Adventures/RadarTowerV2/Effects/M_BuildBase_Emissive
graph {
 properties {
  bUsedWithMeshDeformer=true
  bUsedWithMeshParticles=true
  bUsedWithSkeletalMesh=true
  pixelDepthOffsetMode=PDOM_Legacy
  shadingModel=MSM_Unlit
  usedShadingModels=Unlit
 }
 nodes [
  n1 = VectorParameter() {parameterName=EmissiveColor, defaultValue=<rgba 0 0.7599 1.5 1>}
  n2 = Divide(A=@n1.RGB) {constB=5}
  n3 = Time() {bOverride_Period=true, period=0.2}
  n4 = Time() {bOverride_Period=true, period=0.1}
  n5 = Multiply(A=Multiply(A=Sine(None=@n3), B=Cosine(None=@n4)), B=Sine(None=Time() {bOverride_Period=true, period=0.3}))
  n6 = TextureSample(UVs=Panner(Coordinate=TextureCoordinate()) {speedX=0.125, speedY=0.125}) {texture=/Game/Effects/Fort_Effects/Textures/Tile/T_CloudScape_08}
  n7 = TextureSample(UVs=Panner(Coordinate=TextureCoordinate() {uTiling=0.5, vTiling=0.5}) {speedX=-0.125, speedY=-0.125}) {texture=&T_CloudScape_08}
  Multiply(
   A=Add(A=LinearInterpolate(A=@n2, B=@n1.RGB, Alpha=@n5), B=Multiply(A=@n6.RGB, B=@n7.RGB))
   B=ScalarParameter() {parameterName=EmissiveScalar}
  )
 ]
}
