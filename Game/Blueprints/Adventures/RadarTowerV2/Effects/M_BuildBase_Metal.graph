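# Default Lit metal material for the build base effects of the RadarTowerV2 adventure, built around a faked reflection color.
# A static switch picks either a cubemap sampled by the world reflection vector or a 2D sphere map.
# The sphere map derives its UVs from the view-space reflection vector's XY, bias-scaled with V negated.
# The chosen color feeds only the DiffuseColor input of ConvertFromDiffSpec, with no material output links shown.
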Material /Game/Blueprints/Adventures/RadarTowerV2/Effects/M_BuildBase_Metal
graph {
 properties {pixelDepthOffsetMode=PDOM_Legacy, usedShadingModels=Default Lit}
 nodes [
  n1 = TextureSampleParameterCube!(UVs=ReflectionVectorWS(), Apply View MipBias) {group=reflection, parameterName=Reflection CubeMap, texture=/Game/CharClasses/Technician/Gadgets/BASE/Effects/T_CrystalColor_Tile}
  n2 = Transform(None=ReflectionVectorWS()) {transformSourceType=TRANSFORMSOURCE_World, transformType=TRANSFORM_View}
  n3 = Multiply(A=ConstantBiasScale(None=ComponentMask(None=@n2) {g=true, r=true}), B=Constant2Vector() {g=-1, r=1})
  MaterialFunctionCall(DiffuseColor=StaticSwitchParameter(
    True=@n1.RGB
    False=TextureSampleParameter2D(UVs=@n3) {group=reflection, parameterName=SphereReflectionMap, texture=/Game/Effects/Fort_Effects/Textures/Tile/T_CloudScape_08}.RGB
   ) {group=reflection, parameterName=CubeMap Or ReflectionProbe}) {materialFunction=&ConvertFromDiffSpec}
 ]
}
